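# Sound cue for the exterior ambient bed of the volcano in the Athena map.
# A random node picks one of two looping exterior volcano waves with equal weight.
# It plays through the Athena ambient attenuation and an outdoor-ducking ambient sound class at slightly reduced volume.
# Concurrency allows only one instance and prevents new ones from starting while it plays.
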SoundCue /Figment_S09/Athena/Sounds/Ambient/Volcano/Amb_Volcano_ExteriorBed_Cue
root=Random_0
nodes [
 {id=Random_0, children [WavePlayer_4,WavePlayer_5], properties {weights [1,1]}}
 {id=WavePlayer_4, properties {bLooping=true, soundWaveAssetPtr=&/Amb_Volcano_ExteriorLoop_02}}
 {id=WavePlayer_5, properties {bLooping=true, soundWaveAssetPtr=&/Amb_Volcano_ExteriorLoop_01}}
]
properties {
 attenuationSettings=&Fort_Ambient_Athena_Att
 maxDistance=2097152
 soundClassObject=/Game/Sounds/FortSoundClassesAndModes/Fort_Ambient_DuckOutdoors_Athena_PSM
 volumeMultiplier=0.8
 concurrencyOverrides {maxCount=1, resolutionRule=PreventNew}
}
