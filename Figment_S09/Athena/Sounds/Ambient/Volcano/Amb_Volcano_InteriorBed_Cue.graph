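# Sound cue for the volcano interior ambient bed in Figment Season 9 Athena, playing a looping distant 2D lava caldera wave.
# A single looping wave player feeds a two-input mixer whose first input is empty, with the cue's volume boosted.
# It uses Athena ambient attenuation and a DuckOutdoors sound class, suggesting it ducks outdoor ambience.
# Concurrency allows one instance and prevents new ones from starting.
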SoundCue /Figment_S09/Athena/Sounds/Ambient/Volcano/Amb_Volcano_InteriorBed_Cue
root=Mixer_1
nodes [
 {id=Mixer_1, children [null,WavePlayer_4], properties {inputVolume [1,1]}}
 {id=WavePlayer_4, properties {bLooping=true, soundWaveAssetPtr=&/AMB_Lava_Caldera_Dist_2d_Loop}}
]
properties {
 attenuationSettings=&Fort_Ambient_Athena_Att
 maxDistance=2097152
 soundClassObject=/Game/Sounds/FortSoundClassesAndModes/Fort_Ambient_DuckOutdoors_Athena_PSM
 volumeMultiplier=2
 concurrencyOverrides {maxCount=1, resolutionRule=PreventNew}
}
