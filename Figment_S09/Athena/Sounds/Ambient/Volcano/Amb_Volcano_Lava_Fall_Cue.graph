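# Looping ambient sound cue for lava falls in the Figment S09 Athena volcano, routed to the Athena ambient sound class.
# A random node picks with equal weight among three looping lava fall waves.
# Attenuation override uses a capsule shape with stereo spread, a non-spatialized inner radius and manual reverb send.
# Boosted volume is paired with a concurrency override that caps simultaneous instances, stopping the quietest.
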SoundCue /Figment_S09/Athena/Sounds/Ambient/Volcano/Amb_Volcano_Lava_Fall_Cue
root=Random_1
nodes [
 {id=Random_1, children [WavePlayer_0,WavePlayer_1,WavePlayer_3], properties {weights [1,1,1]}}
 {id=WavePlayer_0, properties {bLooping=true, soundWaveAssetPtr=&/Amb_Volcano_Lava_Fall_01}}
 {id=WavePlayer_1, properties {bLooping=true, soundWaveAssetPtr=/Game/Athena/Sounds/Ambient/Volcano/Amb_Volcano_Lava_Fall_02}}
 {id=WavePlayer_3, properties {bLooping=true, soundWaveAssetPtr=&/Amb_Volcano_Lava_Fall_03}}
]
properties {
 bOverrideAttenuation=true
 bOverrideConcurrency=true
 maxDistance=6000
 soundClassObject=&Fort_Ambient_Athena
 volumeMultiplier=2.5
 attenuationOverrides {
  attenuationShape=Capsule
  falloffDistance=2500
  manualReverbSendLevel=15
  nonSpatializedRadiusStart=700
  reverbSendMethod=Manual
  stereoSpread=1000
  attenuationShapeExtents=<3500 1500>
 }
 concurrencyOverrides {maxCount=4, resolutionRule=StopQuietest}
}
