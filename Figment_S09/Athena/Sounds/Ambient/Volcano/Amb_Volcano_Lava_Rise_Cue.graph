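# Ambient sound cue for rising lava at the volcano, layering looping bed sounds with one-shot accents.
# A close box-attenuated branch mixes looping lava-lake churn and lava-fall waves.
# A wider box-attenuated branch, mixed louder, plays non-looping steam release and flare-up waves together.
# Both branches override attenuation with reverb send disabled; the cue uses an outdoor-ducking ambient sound class and concurrency that stops the quietest instance.
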SoundCue /Figment_S09/Athena/Sounds/Ambient/Volcano/Amb_Volcano_Lava_Rise_Cue
root=Mixer_0
nodes [
 {
  id=Attenuation_0
  children [Mixer_3]
  properties {
   bOverrideAttenuation=true
   attenuationOverrides {attenuationShape=Box, bEnableReverbSend=false, falloffDistance=1024, stereoSpread=512, attenuationShapeExtents=<1024 1024 768>}
  }
 }
 {
  id=Attenuation_1
  children [Mixer_1]
  properties {
   bOverrideAttenuation=true
   attenuationOverrides {attenuationShape=Box, bEnableReverbSend=false, falloffDistance=2048, stereoSpread=512, attenuationShapeExtents=<2048 2048 768>}
  }
 }
 {id=Mixer_0, children [Attenuation_0, Attenuation_1], properties {inputVolume [1,2]}}
 {id=Mixer_1, children [WavePlayer_2,WavePlayer_4], properties {inputVolume [1,1]}}
 {id=Mixer_3, children [WavePlayer_9,WavePlayer_6], properties {inputVolume [1,1]}}
 {id=WavePlayer_2, properties {soundWaveAssetPtr=&/AMB_Lava_Steam_Release_01}}
 {id=WavePlayer_4, properties {soundWaveAssetPtr=&/AMB_Lava_Flare_Up_04}}
 {
  id=WavePlayer_6
  properties {bLooping=true, soundWaveAssetPtr=&/AMB_Lava_Lake_Close_Churn_Splash_Loop}
 }
 {id=WavePlayer_9, properties {bLooping=true, soundWaveAssetPtr=&/Amb_Volcano_Lava_Fall_03}}
]
properties {
 bOverrideConcurrency=true
 maxDistance=4096
 soundClassObject=/Game/Sounds/FortSoundClassesAndModes/Fort_Ambient_DuckOutdoors_Athena_PSM
 volumeMultiplier=2
 attenuationOverrides {
  attenuationShape=Capsule
  falloffDistance=2500
  manualReverbSendLevel=15
  nonSpatializedRadiusStart=700
  reverbSendMethod=Manual
  stereoSpread=1000
  attenuationShapeExtents=<3500 1500>
 }
 concurrencyOverrides {maxCount=4, resolutionRule=StopQuietest}
}
